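# Unlit, translucent, two-sided glow material for Niagara mesh particles in the DaffodilSong backpack FX.
# A mask texture tinted by particle color feeds an emissive adjustment function.
# The mask times particle alpha goes through a depth fade, then is multiplied by a view-angle fade for transparency.
# That fade smoothsteps the absolute camera-to-vertex-normal dot product, softening edge-on faces on both sides.
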Material /Game/Accessories/FORT_Backpacks/Backpack_M_MED_DaffodilSong/FX/Materials/M_Backpack_DaffodilSong_Glow
graph {
 properties {
  bUsedWithNiagaraMeshParticles=true
  blendMode=BLEND_Translucent
  pixelDepthOffsetMode=PDOM_Legacy
  shadingModel=MSM_Unlit
  twoSided=true
  usedShadingModels=Unlit
 }
 nodes [
  n1 = TextureSampleParameter2D() {
   parameterName=Texture
   samplerType=SAMPLERTYPE_Masks
   texture=/Game/Accessories/FORT_Backpacks/Backpack_M_MED_DaffodilSong/FX/Texture/T_Backpack_DaffodilSong_FX_Glow
  }
  n2 = ParticleColor()
  MaterialFunctionCall(Emissive=Multiply(A=@n1.RGB, B=@n2.RGB)) {materialFunction=&MF_FX_EmissiveAdjustments}
  n3 = DotProduct(A=CameraVectorWS(), B=VertexNormalWS())
  Multiply(
   A=DepthFade(Opacity=Multiply(A=@n1.RGB, B=@n2.A), FadeDistance=ScalarParameter() {defaultValue=3, parameterName=DepthFade})
   B=SmoothStep(
    Min=ScalarParameter() {parameterName=View Angle Fade Min, sliderMax=1}
    Max=ScalarParameter() {defaultValue=0.8, parameterName=View Angle Fade Max, sliderMax=1}
    Value=Abs(None=@n3)
   )
  )
 ]
}
